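# Unlit, two-sided translucent trail material for the ChimeCurlCorn character's Niagara ribbon and mesh particle FX.
# Tinted, particle-colored emissive comes from a panning trail texture, with UVs distorted by scrolling smoke noise.
# Opacity combines a similarly panned mask, a sine taper along U, depth fade and view-angle fade.
# World position offset blends SplineThicken with an optional sine wave; dynamic parameters control width, UV scale, scroll and wave.
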Material /BRCosmetics/Characters/Player/Female/Medium/Bodies/F_MED_ChimeCurlCorn/FX/M_Character_ChimeCurlCorn_Trail
graph {
 properties {
  bUsedWithNiagaraMeshParticles=true
  bUsedWithNiagaraRibbons=true
  blendMode=BLEND_Translucent
  pixelDepthOffsetMode=PDOM_Legacy
  shadingModel=MSM_Unlit
  twoSided=true
  usedShadingModels=Unlit
 }
 nodes [
  n1 = ScalarParameter() {defaultValue=25, parameterName=Width}
  n2 = LinearInterpolate(
   A=@n1
   B=Multiply(A=@n1, B=DynamicParameter() {paramNames [WidthScale, TexSpeed_Offset, WPO, WPO_SpeedOffset]}.WidthScale)
   Alpha=ScalarParameter() {parameterName=UseCustomWidth}
  )
  n3 = StaticBoolParameter() {parameterName="Expand U or V UV Channel (SB)"}
  n4 = VectorParameter() {parameterName=WPO UV Scale, defaultValue=<rgba 1.2 1 0 0>}
  n5 = DynamicParameter() {parameterIndex=1, paramNames [UVScale,WPO_Scale,Param3,Param4]}
  n6 = Multiply(A=TextureCoordinate(), B=AppendVector(A=Multiply(A=@n4.R, B=@n5.WPO_Scale), B=@n4.G))
  n7 = DynamicParameter() {paramNames [WidthScale, TexSpeed_Offset, WPO, WPO_SpeedOffset]}
  n8 = ScalarParameter() {defaultValue=-2, parameterName=WPO_Speed}
  n9 = Add(A=Add(A=ComponentMask(None=@n6) {r=true}, B=@n7.WPO_SpeedOffset), B=Multiply(A=Time(), B=Multiply(A=@n8, B=@n7.WPO_SpeedOffset)))
  n10 = ComponentMask(None=TextureCoordinate()) {r=true}
  n11 = Multiply(A=LinearInterpolate(Alpha=Sine(None=@n9)) {constA=-0.2}, B=Power(Base=Sine(None=@n10) {period=2}))
  n12 = Multiply(
   A=Multiply(A=Constant3Vector() {constant=<rgba 0 1 0 0>}, B=Multiply(A=@n11) {constB=0.3})
   B=DynamicParameter() {paramNames [WidthScale, TexSpeed_Offset, WPO, WPO_SpeedOffset]}.WPO
  )
  Add(
   A=LinearInterpolate(
    B=MaterialFunctionCall(WidthBase=@n2, WidthTip=@n2, Expand U or V UV Channel=@n3) {materialFunction=&SplineThicken}.WorldPositionOffset
    Alpha=ScalarParameter() {defaultValue=1, parameterName=Use_SplineThicken}
   )
   B=LinearInterpolate(B=Transform(None=@n12), Alpha=ScalarParameter() {parameterName=Use_WPO})
  )
  n13 = ParticleColor()
  n14 = Desaturation(
   None=VectorParameter() {parameterName=Color, defaultValue=<rgba 1.5 0.4388 0.3252 1>}.RGB
   Fraction=ScalarParameter() {parameterName=DefaultColorDes}
  )
  n15 = TextureCoordinate()
  n16 = Multiply(
   A=TextureSample(UVs=Panner(Coordinate=@n15) {speedX=0.1}) {texture=&T_Inky_Smoke_Tile}.R
   B=ScalarParameter() {parameterName=NoiseInt}
  ) {constB=0.15}
  n17 = VectorParameter() {parameterName=BaseTextureUV, defaultValue=<rgba 1 1 1 1>}
  n18 = DynamicParameter() {parameterIndex=1, paramNames [UVScale,WPO_Scale,Param3,Param4]}
  n19 = Multiply(A=TextureCoordinate(), B=AppendVector(A=Multiply(A=@n17.R, B=@n18.UVScale), B=@n17.G))
  n20 = DynamicParameter() {paramNames [WidthScale, TexSpeed_Offset, WPO, WPO_SpeedOffset]}
  n21 = Constant()
  n22 = TextureSampleParameter2D(UVs=Panner(Coordinate=Add(A=@n16, B=@n19), Speed=AppendVector(A=@n20.TexSpeed_Offset, B=@n21)) {speedX=-0.8}) {
   parameterName=BaseTexture
   texture=&/T_Character_ChimeCurlCorn_Trail
  }
  n23 = Power(Base=@n22.R, Exp=ScalarParameter() {defaultValue=1.5, parameterName=BaseTexture_Power}) {constExponent=1.5}
  MaterialFunctionCall(Emissive=Multiply(A=@n13.RGB, B=Multiply(A=@n14, B=@n23))) {materialFunction=&MF_FX_EmissiveAdjustments}
  n24 = VectorParameter() {parameterName=MaskTextureUV, defaultValue=<rgba 1 1 1 1>}
  n25 = DynamicParameter() {parameterIndex=1, paramNames [UVScale,WPO_Scale,Param3,Param4]}
  n26 = Multiply(A=TextureCoordinate(), B=AppendVector(A=Multiply(A=@n24.R, B=@n25.UVScale), B=@n24.G))
  n27 = DynamicParameter() {paramNames [WidthScale, TexSpeed_Offset, WPO, WPO_SpeedOffset]}
  n28 = Constant()
  n29 = TextureSampleParameter2D(UVs=Panner(Coordinate=Add(A=@n26, B=@n16), Speed=AppendVector(A=@n27.TexSpeed_Offset, B=@n28)) {speedX=-1}) {parameterName=Mask, texture=&T_Character_ChimeCurlCorn_Trail}
  n30 = TextureCoordinate()
  n31 = Multiply(A=Add(A=@n29.R) {constB=0}, B=Sine(None=ComponentMask(None=@n30) {r=true}) {period=2})
  n32 = MaterialFunctionCall("FadeSensitivity (optional)"=ScalarParameter() {parameterName="FadeSensitivity (optional) (S)"}) {materialFunction=&FadeBasedOnViewAngle}
  Multiply(
   A=DepthFade(Opacity=Multiply(A=@n13.A, B=@n31), FadeDistance=ScalarParameter() {defaultValue=15, parameterName=FadeDistance}) {fadeDistanceDefault=10}
   B=Saturate(None=Multiply(A=@n32.Result) {constB=10})
  )
 ]
}
