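# Idle cosmetic Niagara effect for the ChimeCurlCorn character body: red audio-reactive glows, rings and stars pinned to skeletal mesh sockets.
# Partition indices pick each particle's socket, offset and scale for lights, looping pulses, two limb ring-mesh layers and stars that spawn trailing sparkles.
# Event audio amplitude, raised to high powers, spikes brightness and star size, while an interpolated system value drives fades.
# User flags handle hidden hands or feet and front-end preview.
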NiagaraSystem /BRCosmetics/Characters/Player/Female/Medium/Bodies/F_MED_ChimeCurlCorn/FX/NS_Character_ChimeCurlCorn_Idle
dependencies {
 usedDataInterfaces [
  &NiagaraDataInterfaceCurve
  &NiagaraDataInterfaceArrayFloat
  &NiagaraDataInterfaceArrayInt32
  &NiagaraDataInterfaceArrayFloat3
  &NiagaraDataInterfaceColorCurve
  &NiagaraDataInterfaceParticleRead
  &NiagaraDataInterfaceSkeletalMesh
 ]
 usedDynamicInputs []
 usedModules [
  &self:SystemSpawnScript.SystemScriptSource.SystemScriptGraph.NiagaraNodeAssignment_1.SetVariables_#1
  &SystemState
  &InterpolateOverTime
  &self:SystemSpawnScript.SystemScriptSource.SystemScriptGraph.NiagaraNodeAssignment_0.SetVariables_#2
 ]
}
emitters {
 Light {
  emitterUpdateScript [
   &EmitterState(
    Inactive Response=::NewEnumerator1
    Life Cycle Mode=::NewEnumerator1
    Loop Behavior=::NewEnumerator1
    Loop Duration Mode=::NewEnumerator1
    UseLoopDelay=false
   )
   &SpawnBurst_Instantaneous(Spawn Count=4, Use Spawn Probability=false, UseLoopCountLimit=false)
  ]
  particleSpawnScript [
   &PartitionParticles(Number Of Partitions=4)
   &InitializeParticle(Sprite Size Mode=::NewEnumerator3, UsePositionOffset=false, Write Material Random=true)
   &CameraOffset(Camera Offset Amount=5)
  ]
  particleUpdateScript [
   &ParticleState(Kill Particles When Lifetime Has Elapsed=false, Let Infinitely Lived Particles Die When Emitter Deactivates=true)
   &SolveForcesAndVelocity(Clamp Velocity=false, Limit Acceleration=false)
   &SkeletalMeshLocation(
    Bone / Socket Index=&SelectIntFromArray(Direct Array Index=$Particles.PartitionParticles.Partition, Int Selection Array=DataInterface() {propertyValues {IntData [3,5,7,9]}})
    Bone Sampling Mode=::NewEnumerator4
    MaskSpawnGroup=false
    Skeletal Mesh=$System.Skeletal Mesh Reader
   )
   &FN_Light_Attributes(
    Color=<rgba 1 0.0872 0.0781 1>
    Create Color Attribute=true
    Create Exponent Attribute=true
    Create Position Attribute=true
    Create Radius Attribute=true
    Exponent=2
    Position:NiagaraPosition=$Particles.SkeletalMeshLocation.SampledPosition
    Radius=&Multiply_Float(A=80, B=&Lerp_Float(A=0, B=1, Alpha=$System.InterpolateOverTime.InterpolateOverTime_Float.Moving Average))
    Scale Color=&Lerp_Float(
     A=20
     B=80
     Alpha=&Power(a=&Multiply_Float(A=$System.Audio Reactive, B=$NPC.FortniteNPC.EventAudio_Amplitude), b=10)
    )
   )
  ]
  rendererClasses [&NiagaraSpriteRendererProperties, &NiagaraLightRendererProperties]
  renderers [
   {rendererClass=&NiagaraSpriteRendererProperties}
   {rendererClass=&NiagaraLightRendererProperties, rendererIndex=1}
  ]
 }
 Pulse {
  emitterUpdateScript [
   &EmitterState(Inactive Response=::NewEnumerator1, Life Cycle Mode=::NewEnumerator1, Loop Duration=0.5, UseLoopDelay=false)
   &SpawnBurst_Instantaneous(Spawn Count=10, Use Spawn Probability=false, UseLoopCountLimit=false)
   &SpawnBurst_Instantaneous(Spawn Count=10, Spawn Time=0.2, Use Spawn Probability=false, UseLoopCountLimit=false) {module=SpawnBurst_Instantaneous001}
  ]
  particleSpawnScript [
   &PartitionParticles(Number Of Partitions=10)
   &InitializeParticle(
    Color=<rgba 2 0.1354 0.1668 20>
    Lifetime=0.5
    Sprite Size Mode=::NewEnumerator3
    Uniform Sprite Size=1
    UsePositionOffset=false
    Write Material Random=true
   )
   &CameraOffset(Camera Offset Amount=&MakeCustomFloatFromBool(Bool=$User.bIsFrontEndPreview, True Float=0, False Float=20))
   /SetVariables(Particles.VisibilityTag=&MakeCustomIntFromBool(Bool=$User.bIsFrontEndPreview, True INT=1, False INT=0)) {isSetParameters=true}
  ]
  particleUpdateScript [
   &ParticleState()
   &ScaleSpriteSize(
    Scale Sprite Size Mode=::NewEnumerator0
    Uniform Scale Factor=&MakeCustomFloatFromBool(
     Bool=$User.bisFootHidden
     True Float=&SelectFloatFromArray(
      Direct Array Index=$Particles.PartitionParticles.Partition
      Float Selection Array=DataInterface() {propertyValues {FloatData [0,0,30,30,30,30,35,0,35,0]}}
     )
     False Float=&SelectFloatFromArray(
      Direct Array Index=$Particles.PartitionParticles.Partition
      Float Selection Array=DataInterface() {propertyValues {FloatData [0,0,30,30,30,30,35,30,35,30]}}
     )
    )
   )
   &ScaleSpriteSize(
    Scale Sprite Size Mode=::NewEnumerator0
    Uniform Scale Factor=&MakeCustomFloatFromBool(
     Bool=$User.bisHandHidden
     True Float=&SelectFloatFromArray(
      Direct Array Index=$Particles.PartitionParticles.Partition
      Float Selection Array=DataInterface() {propertyValues {FloatData [0,0,30,0,30,0,35,30,35,30]}}
     )
     False Float=&SelectFloatFromArray(
      Direct Array Index=$Particles.PartitionParticles.Partition
      Float Selection Array=DataInterface() {propertyValues {FloatData [0,0,30,30,30,30,35,30,35,30]}}
     )
    )
   ) {enabled=false, module=ScaleSpriteSize001}
   &SkeletalMeshLocation(
    Bone / Socket Index=&SelectIntFromArray(
     Direct Array Index=$Particles.PartitionParticles.Partition
     Direct Array Mode=::NewEnumerator1
     Int Selection Array=DataInterface() {propertyValues {IntData [0,1,2,3,4,5,6,7,8,9]}}
    )
    Bone Sampling Mode=::NewEnumerator4
    MaskSpawnGroup=false
    Sampled Position Offset=&SelectVectorFromArray(
     Direct Array Index=$Particles.PartitionParticles.Partition
     Vector Selection Array=DataInterface() {propertyValues {FloatData [<0 0 0>, <0 0 0>, <5 0 0>, <0 0 0>, <-5 0 0>, <0 0 0>, <-30 0 0>, <0 0 0>, <30 0 0>, <0 0 0>]}}
    )
    Skeletal Mesh=$System.Skeletal Mesh Reader
   )
   /SetVariables(Particles.OffsetPosition:NiagaraPosition=&AddVectorToPosition(
     Position=$Particles.Position
     Vector=&RotateVectorByQuaternion(Vector To Rotate=<0 0 0>, Quaternion=$Particles.SkeletalMeshLocation.SampledRotation)
    )) {isSetParameters=true}
   &DynamicMaterialParameters(
    Index 0 Param 1=&MakeFloatFromVector(Vector=Linked() {linkedVariable {name=Particles.Position, type=NiagaraPosition}}, Channel=::NewEnumerator0)
    Index 0 Param 2=&MakeFloatFromVector(Vector=Linked() {linkedVariable {name=Particles.Position, type=NiagaraPosition}}, Channel=::NewEnumerator1)
    Index 0 Param 3=&MakeFloatFromVector(Vector=Linked() {linkedVariable {name=Particles.Position, type=NiagaraPosition}}, Channel=::NewEnumerator2)
    Index 1 Param 1=&MakeFloatFromVector(Vector=Linked() {linkedVariable {name=Particles.OffsetPosition, type=NiagaraPosition}}, Channel=::NewEnumerator0)
    Index 1 Param 2=&MakeFloatFromVector(Vector=Linked() {linkedVariable {name=Particles.OffsetPosition, type=NiagaraPosition}}, Channel=::NewEnumerator1)
    Index 1 Param 3=&MakeFloatFromVector(Vector=Linked() {linkedVariable {name=Particles.OffsetPosition, type=NiagaraPosition}}, Channel=::NewEnumerator2)
    Index 1 Param 4=&Lerp_Float(
     A=&FloatFromCurve(
      FloatCurve=DataInterface() {propertyValues {Curve {keys [
          {arriveTangent=-0.0762, interpMode=RCIM_Cubic, leaveTangent=-0.0762, tangentMode=RCTM_User, time=0, value=0.8}
          {arriveTangent=-0.2085, interpMode=RCIM_Cubic, leaveTangent=-0.2085, tangentMode=RCTM_User, time=1, value=0.3}
         ]}}}
      Scale Curve=1.35
     )
     B=&FloatFromCurve(
      FloatCurve=DataInterface() {propertyValues {Curve {keys [
          {arriveTangent=-0.8006, interpMode=RCIM_Cubic, leaveTangent=-0.8006, tangentMode=RCTM_User, time=0, value=0.8}
          {arriveTangent=-0.3233, interpMode=RCIM_Cubic, leaveTangent=-0.3233, tangentMode=RCTM_User, time=1, value=0.4}
         ]}}}
      Scale Curve=1.35
     )
     Alpha=$System.InterpolateOverTime.InterpolateOverTime_Float.Moving Average
    )
    Module.Index0Float0Write=true
    Module.Index0Float1Write=true
    Module.Index0Float2Write=true
    Module.Index0Float3Write=false
    Module.Index1Float0Write=true
    Module.Index1Float1Write=true
    Module.Index1Float2Write=true
    Module.Index1Float3Write=true
    Write Parameter Index 1=::NewEnumerator1
   )
   &ScaleColor(
    Scale Alpha=&FloatFromCurve(
     FloatCurve=DataInterface() {propertyValues {Curve {keys [
         {arriveTangent=0.5632, interpMode=RCIM_Cubic, leaveTangent=0.5632, tangentMode=RCTM_User, time=0, value=0}
         {arriveTangent=0.168, interpMode=RCIM_Cubic, leaveTangent=0.168, tangentMode=RCTM_User, time=0.25, value=1}
         {arriveTangent=-0.1012, interpMode=RCIM_Cubic, leaveTangent=-0.1012, tangentMode=RCTM_User, time=1, value=0}
        ]}}}
     Scale Curve=$System.InterpolateOverTime.InterpolateOverTime_Float.Moving Average
    )
    Scale RGB=&VectorFromFloat(Value=&FloatFromCurve(FloatCurve=DataInterface() {propertyValues {Curve {keys [
          {arriveTangent=0.168, interpMode=RCIM_Cubic, leaveTangent=0.168, tangentMode=RCTM_User, time=0.2, value=1.5}
          {arriveTangent=-0.1012, interpMode=RCIM_Cubic, leaveTangent=-0.1012, tangentMode=RCTM_User, time=0.4, value=1}
         ]}}}))
    ScaleA=true
    ScaleRGB=true
   )
  ]
  rendererClasses [&NiagaraSpriteRendererProperties]
  renderers [
   {rendererClass=&NiagaraSpriteRendererProperties}
   {rendererClass=&NiagaraSpriteRendererProperties, rendererIndex=1}
   {rendererClass=&NiagaraSpriteRendererProperties, rendererIndex=2}
  ]
 }
 Ring_Limbs {
  emitterUpdateScript [
   &EmitterState(
    Inactive Response=::NewEnumerator1
    Life Cycle Mode=::NewEnumerator1
    Loop Behavior=::NewEnumerator1
    Loop Duration Mode=::NewEnumerator1
    UseLoopDelay=false
   )
   &SpawnBurst_Instantaneous(Spawn Count=8, Use Spawn Probability=false, UseLoopCountLimit=false)
   &SpawnRate(SpawnRate=10, Use Spawn Probability=false) {enabled=false}
  ]
  particleSpawnScript [
   &PartitionParticles(Number Of Partitions=8)
   &InitializeParticle(Color Mode=::NewEnumerator0, Lifetime=0.6, Mesh Scale Mode=::NewEnumerator4, UsePositionOffset=false, Write Material Random=true)
   &InitialMeshOrientation(
    Mesh Orientation Mode=::NewEnumerator2
    Rotation=&SelectVectorFromArray(
     Direct Array Index=$Particles.PartitionParticles.Partition
     Vector Selection Array=DataInterface() {propertyValues {FloatData [<0 0 0.5>, <0 0 0.5>, <0 0 0>, <0 0 0>, <0 0 0>, <0 0 0>, <0 0 0.5>, <0 0 0.5>]}}
    )
    Use Rotation Vector=true
   )
   &InitialMeshOrientation(
    Mesh Orientation Mode=::NewEnumerator2
    Rotation=&SelectVectorFromArray(
     Direct Array Index=$Particles.PartitionParticles.Partition
     Vector Selection Array=DataInterface() {propertyValues {FloatData [<0.6 0 0>, <0 0 0>, <0.5 0 0>, <0.3 0 0>, <0 0 0>, <0 0 0>, <0.1 0 0>, <0 0 0>]}}
    )
    Use Rotation Vector=true
   ) {module=InitialMeshOrientation001}
   &InitialMeshOrientation(
    Mesh Orientation Mode=::NewEnumerator2
    Rotation=&SelectVectorFromArray(
     Direct Array Index=$Particles.PartitionParticles.Partition
     Vector Selection Array=DataInterface() {propertyValues {FloatData [<0 0 0>, <0 0 0>, <0 0 0>, <0 0 0>, <0 0 0>, <0 0 0>, <0 0 0>, <0 0 0>]}}
    )
    Use Rotation Vector=true
   ) {module=InitialMeshOrientation002}
   &CameraOffset(Camera Offset Amount=10)
  ]
  particleUpdateScript [
   &ParticleState(Kill Particles When Lifetime Has Elapsed=false, Let Infinitely Lived Particles Die When Emitter Deactivates=true)
   &ScaleMeshSize(Scale Factor=&Multiply_Vector(
     A=&MakeCustomVectorFromBool(
      Bool=$User.bisFootHidden
      True Vector=&SelectVectorFromArray(
       Direct Array Index=$Particles.PartitionParticles.Partition
       Vector Selection Array=DataInterface() {propertyValues {FloatData [<0.8 1 -1>, <0.7 0.7 0.7>, <0.8 0.9 0.9>, <0.7 0.7 -0.7>, <1.1 1 -1>, <0 0 0>, <1.2 1 1>, <0 0 0>]}}
      )
      False Vector=&SelectVectorFromArray(
       Direct Array Index=$Particles.PartitionParticles.Partition
       Vector Selection Array=DataInterface() {propertyValues {FloatData [<0.8 1 -1>, <0.7 0.7 0.7>, <0.8 0.9 0.9>, <0.7 0.7 -0.7>, <1.1 1 -1>, <0.6 0.8 -0.8>, <1.2 1 1>, <0.6 0.8 0.8>]}}
      )
     )
     B=&RandomRangeVector(Minimum=<0.8 1 1>, Maximum=<1 1 1>)
    )) {module=ScaleMeshSize002}
   &ScaleMeshSize(Scale Factor=&Multiply_Vector(
     A=&MakeCustomVectorFromBool(
      Bool=$User.bisHandHidden
      True Vector=&SelectVectorFromArray(
       Direct Array Index=$Particles.PartitionParticles.Partition
       Vector Selection Array=DataInterface() {propertyValues {FloatData [<1 1 1>, <0 0 0>, <1 1 1>, <0 0 0>, <1 1 1>, <1 1 1>, <1 1 1>, <1 1 1>]}}
      )
      False Vector=<1 1 1>
     )
     B=&RandomRangeVector(Minimum=<0.8 1 1>, Maximum=<1 1 1>)
    )) {module=ScaleMeshSize003}
   &ScaleMeshSize(Scale Factor=&MakeCustomVectorFromBool(Bool=$User.bIsFrontEndPreview, True Vector=<1 0.8 0.8>, False Vector=<1 1 1>))
   &SkeletalMeshLocation(
    Bone / Socket Index=&SelectIntFromArray(Direct Array Index=$Particles.PartitionParticles.Partition, Int Selection Array=DataInterface() {propertyValues {IntData [2,3,4,5,6,7,8,9]}})
    Bone Sampling Mode=::NewEnumerator4
    MaskSpawnGroup=false
    Sampled Position Offset=&SelectVectorFromArray(
     Direct Array Index=$Particles.PartitionParticles.Partition
     Vector Selection Array=DataInterface() {propertyValues {FloatData [<25 2 2>, <20 0 2>, <-25 2 -2>, <-25 0 1>, <-37 0 -3>, <-20 0 -2>, <30 0 3>, <20 0 2>]}}
    )
    Skeletal Mesh=$System.Skeletal Mesh Reader
   )
   &DynamicMaterialParameters(
    Index 0 Param 1=&SelectFloatFromArray(
     Direct Array Index=$Particles.PartitionParticles.Partition
     Float Selection Array=DataInterface() {propertyValues {FloatData [13,10,13,10,20,12,20,15]}}
    )
    Index 0 Param 2=&SelectFloatFromArray(
     Direct Array Index=$Particles.PartitionParticles.Partition
     Float Selection Array=DataInterface() {propertyValues {FloatData [-0.5,-0.2,-1,-0.3,-1.3,-0.6,-0.5,-1.2]}}
    )
    Index 0 Param 3=&Lerp_Float(
     A=2
     B=20
     Alpha=&Power(a=&Multiply_Float(A=1, B=$NPC.FortniteNPC.EventAudio_Amplitude), b=10)
    )
    Index 0 Param 4=&RandomRangeFloat(Minimum=0.8, Maximum=1.2)
    Index 1 Param 1=&SelectFloatFromArray(
     Direct Array Index=$Particles.PartitionParticles.Partition
     Float Selection Array=DataInterface() {propertyValues {FloatData [1.5,1,1,1,1.3,0.9,1.2,1]}}
    )
    Index 1 Param 2=&SelectFloatFromArray(
     Direct Array Index=$Particles.PartitionParticles.Partition
     Float Selection Array=DataInterface() {propertyValues {FloatData [1,1,1.3,1,1.5,0.5,1.2,1]}}
    )
    Module.Index0Param0Write=true
    Module.Index0Param1Write=true
    Module.Index0Param2Write=true
    Module.Index0Param3Write=true
    Module.Index1Param0Write=true
    Module.Index1Param1Write=true
    Module.Index1Param2Write=false
    Module.Index1Param3Write=false
    Write Parameter Index 1=true
   )
   &SolveForcesAndVelocity(Clamp Velocity=false, Limit Acceleration=false)
   /SetVariables(Particles.ColorBase:LinearColor=&MakeLinearColorFromVectorAndFloat(
     "Vector (RGB)"=&MakeVectorFromLinearColorRGB(Color=<rgba 0.6823 0.0746 0.0867 1>)
     "Float (Alpha)"=&FloatFromCurve(FloatCurve=DataInterface() {propertyValues {Curve {keys [<at 0 0 RCIM_Cubic>,<at 0.4 1 RCIM_Cubic>,<at 1 0 RCIM_Cubic>]}}}, Scale Curve=1.5)
    )) {enabled=false, isSetParameters=true, module=SetVariables_#3}
   &ScaleColor(
    Scale Alpha=&MakeCustomFloatFromBool(
     Bool=$User.bIsFrontEndPreview
     True Float=&Lerp_Float(A=0.5, B=1, Alpha=$System.InterpolateOverTime.InterpolateOverTime_Float.Moving Average)
     False Float=&Lerp_Float(A=0, B=1, Alpha=$System.InterpolateOverTime.InterpolateOverTime_Float.Moving Average)
    )
    Scale RGB=&MakeVectorFromLinearColorRGB(Color=&Multiply_LinearColorByFloat(
      Linear Color=<rgba 1 0.0891 0.1273 1>
      Float=&Lerp_Float(
       A=2
       B=35
       Alpha=&Power(a=&Multiply_Float(A=$System.Audio Reactive, B=$NPC.FortniteNPC.EventAudio_Amplitude), b=10)
      )
     ))
    ScaleA=true
    ScaleRGB=true
   )
  ]
  rendererClasses [&NiagaraMeshRendererProperties]
  renderers [
   {rendererClass=&NiagaraMeshRendererProperties, rendererIndex=0}
   {rendererClass=&NiagaraMeshRendererProperties, rendererIndex=1}
  ]
 }
 Ring_Limbs_small {
  emitterUpdateScript [
   &EmitterState(
    Inactive Response=::NewEnumerator1
    Life Cycle Mode=::NewEnumerator1
    Loop Behavior=::NewEnumerator1
    Loop Duration Mode=::NewEnumerator1
    UseLoopDelay=false
   )
   &SpawnBurst_Instantaneous(Spawn Count=8, Use Spawn Probability=false, UseLoopCountLimit=false)
   &SpawnRate(SpawnRate=10, Use Spawn Probability=false) {enabled=false}
  ]
  particleSpawnScript [
   &PartitionParticles(Number Of Partitions=8)
   &InitializeParticle(Color=<rgba 1 1 1 1.5>, Lifetime=0.6, Mesh Scale Mode=::NewEnumerator4, UsePositionOffset=false, Write Material Random=true)
   &InitialMeshOrientation(
    Mesh Orientation Mode=::NewEnumerator2
    Rotation=&SelectVectorFromArray(
     Direct Array Index=$Particles.PartitionParticles.Partition
     Vector Selection Array=DataInterface() {propertyValues {FloatData [<0 0 0.5>, <0 0 0.5>, <0 0 0>, <0 0 0>, <0 0 0>, <0 0 0>, <0 0 0.5>, <0 0 0.5>]}}
    )
    Use Rotation Vector=true
   )
   &InitialMeshOrientation(
    Mesh Orientation Mode=::NewEnumerator2
    Rotation=&SelectVectorFromArray(
     Direct Array Index=$Particles.PartitionParticles.Partition
     Vector Selection Array=DataInterface() {propertyValues {FloatData [<0.4 0 0>, <0.6 0.1 0>, <0.4 0 0>, <0 0.02 -0.03>, <-0.3 0 0>, <-0.3 0 0>, <-0.2 0 0>, <0 0 0>]}}
    )
    Use Rotation Vector=true
   ) {module=InitialMeshOrientation001}
   &InitialMeshOrientation(
    Mesh Orientation Mode=::NewEnumerator2
    Rotation=&SelectVectorFromArray(
     Direct Array Index=$Particles.PartitionParticles.Partition
     Vector Selection Array=DataInterface() {propertyValues {FloatData [<0 0 0>, <0 0 0>, <0 0 0>, <0 0 0>, <0 0 0>, <0 0 0>, <0 0 0>, <0 0 0>]}}
    )
    Use Rotation Vector=true
   ) {module=InitialMeshOrientation002}
   &CameraOffset(Camera Offset Amount=5)
  ]
  particleUpdateScript [
   &ParticleState(Kill Particles When Lifetime Has Elapsed=false, Let Infinitely Lived Particles Die When Emitter Deactivates=true)
   &ScaleMeshSize(Scale Factor=&Multiply_Vector(
     A=&MakeCustomVectorFromBool(
      Bool=$User.bisFootHidden
      True Vector=&SelectVectorFromArray(
       Direct Array Index=$Particles.PartitionParticles.Partition
       Vector Selection Array=DataInterface() {propertyValues {FloatData [<0.5 0.8 -0.8>, <0.5 0.6 0.6>, <0.6 0.6 0.6>, <0.6 0.6 0.6>, <1 1 1>, <0 0 0>, <1 1 1>, <0 0 0>]}}
      )
      False Vector=&SelectVectorFromArray(
       Direct Array Index=$Particles.PartitionParticles.Partition
       Vector Selection Array=DataInterface() {propertyValues {FloatData [<0.5 0.8 -0.8>, <0.5 0.6 0.6>, <0.6 0.6 0.6>, <0.6 0.6 -0.6>, <1 1 1>, <0.7 0.8 -0.8>, <1 1 1>, <0.6 0.8 -0.8>]}}
      )
     )
     B=&RandomRangeVector(Minimum=<0.8 1 1>, Maximum=<1 1 1>)
    )) {module=ScaleMeshSize002}
   &ScaleMeshSize(Scale Factor=&Multiply_Vector(
     A=&MakeCustomVectorFromBool(
      Bool=$User.bisHandHidden
      True Vector=&SelectVectorFromArray(
       Direct Array Index=$Particles.PartitionParticles.Partition
       Vector Selection Array=DataInterface() {propertyValues {FloatData [<1 1 1>, <0 0 0>, <1 1 1>, <0 0 0>, <1 1 1>, <1 1 1>, <1 1 1>, <1 1 1>]}}
      )
      False Vector=<1 1 1>
     )
     B=&RandomRangeVector(Minimum=<0.8 1 1>, Maximum=<1 1 1>)
    )) {module=ScaleMeshSize001}
   &ScaleMeshSize(Scale Factor=&MakeCustomVectorFromBool(Bool=$User.bIsFrontEndPreview, True Vector=<1 0.8 0.8>, False Vector=<1 1 1>))
   &SkeletalMeshLocation(
    Bone / Socket Index=&SelectIntFromArray(Direct Array Index=$Particles.PartitionParticles.Partition, Int Selection Array=DataInterface() {propertyValues {IntData [2,3,4,5,6,7,8,9]}})
    Bone Sampling Mode=::NewEnumerator4
    MaskSpawnGroup=false
    Sampled Position Offset=&SelectVectorFromArray(
     Direct Array Index=$Particles.PartitionParticles.Partition
     Vector Selection Array=DataInterface() {propertyValues {FloatData [<25 0 0>, <20 1 0>, <-25 2 -2>, <-25 1 0.5>, <-35 0 2>, <-20 -2 1>, <40 2 0>, <23 0 -2>]}}
    )
    Skeletal Mesh=$System.Skeletal Mesh Reader
   )
   &DynamicMaterialParameters(
    Index 0 Param 1=&SelectFloatFromArray(
     Direct Array Index=$Particles.PartitionParticles.Partition
     Float Selection Array=DataInterface() {propertyValues {FloatData [8,4,5,5,8,8,12,8]}}
    )
    Index 0 Param 2=&SelectFloatFromArray(
     Direct Array Index=$Particles.PartitionParticles.Partition
     Float Selection Array=DataInterface() {propertyValues {FloatData [-0.7,-0.1,-0.8,-0.2,-1,-0.8,-1,-0.6]}}
    )
    Index 0 Param 3=&Lerp_Float(
     A=3
     B=10
     Alpha=&Power(a=&Multiply_Float(A=1, B=$NPC.FortniteNPC.EventAudio_Amplitude), b=10)
    )
    Index 0 Param 4=&RandomRangeFloat(Minimum=0.5, Maximum=1)
    Index 1 Param 1=&SelectFloatFromArray(
     Direct Array Index=$Particles.PartitionParticles.Partition
     Float Selection Array=DataInterface() {propertyValues {FloatData [1.5,0.7,1.5,1,1,0.9,1.5,1]}}
    )
    Index 1 Param 2=&SelectFloatFromArray(
     Direct Array Index=$Particles.PartitionParticles.Partition
     Float Selection Array=DataInterface() {propertyValues {FloatData [1.5,1,1.5,1,0.5,0.5,1.2,1]}}
    )
    Module.Index0Param0Write=true
    Module.Index0Param1Write=true
    Module.Index0Param2Write=true
    Module.Index0Param3Write=true
    Module.Index1Param0Write=true
    Module.Index1Param1Write=true
    Module.Index1Param2Write=false
    Module.Index1Param3Write=false
    Write Parameter Index 1=true
   )
   &SolveForcesAndVelocity(Clamp Velocity=false, Limit Acceleration=false)
   /SetVariables(Particles.ColorBase:LinearColor=&MakeLinearColorFromVectorAndFloat(
     "Vector (RGB)"=&MakeVectorFromLinearColorRGB(Color=<rgba 0.6823 0.0746 0.0867 1>)
     "Float (Alpha)"=&FloatFromCurve(FloatCurve=DataInterface() {propertyValues {Curve {keys [<at 0 0 RCIM_Cubic>,<at 0.4 1 RCIM_Cubic>,<at 1 0 RCIM_Cubic>]}}}, Scale Curve=1.5)
    )) {enabled=false, isSetParameters=true, module=SetVariables_#3}
   &ScaleColor(
    Scale Alpha=&MakeCustomFloatFromBool(
     Bool=$User.bIsFrontEndPreview
     True Float=&Lerp_Float(A=0.8, B=1, Alpha=$System.InterpolateOverTime.InterpolateOverTime_Float.Moving Average)
     False Float=&Lerp_Float(A=0, B=1, Alpha=$System.InterpolateOverTime.InterpolateOverTime_Float.Moving Average)
    )
    Scale RGB=&MakeVectorFromLinearColorRGB(Color=&Multiply_LinearColorByFloat(
      Linear Color=<rgba 0.3854 0 0.0307 1>
      Float=&Lerp_Float(
       A=5
       B=30
       Alpha=&Power(a=&Multiply_Float(A=$System.Audio Reactive, B=$NPC.FortniteNPC.EventAudio_Amplitude), b=10)
      )
     ))
    ScaleA=true
    ScaleRGB=true
   )
  ]
  rendererClasses [&NiagaraMeshRendererProperties]
  renderers [
   {rendererClass=&NiagaraMeshRendererProperties, rendererIndex=0}
   {rendererClass=&NiagaraMeshRendererProperties, rendererIndex=1}
  ]
 }
 Star {
  emitterUpdateScript [
   &EmitterState(Life Cycle Mode=::NewEnumerator1, Loop Duration=1, UseLoopDelay=false)
   &SpawnRate(
    SpawnRate=&MakeCustomFloatFromBool(Bool=$User.bisFootHidden, True Float=2, False Float=4)
    Use Spawn Probability=false
   )
  ]
  particleSpawnScript [
   &PartitionParticles(Number Of Partitions=4)
   &InitializeParticle(
    Color=<rgba 0.6667 0 0 2>
    Lifetime=0.6
    Sprite Rotation Mode=::NewEnumerator2
    Sprite Size Mode=::NewEnumerator1
    Uniform Sprite Size Max=13
    Uniform Sprite Size Min=10
    UsePositionOffset=false
    Write Material Random=true
   )
   &CameraOffset(Camera Offset Amount=20)
   &SkeletalMeshLocation(
    Bone / Socket Index=&MakeCustomIntFromBool(
     Bool=$User.bisFootHidden
     True INT=&SelectIntFromArray(Direct Array Index=$Particles.PartitionParticles.Partition, Int Selection Array=DataInterface() {propertyValues {IntData [2,4,2,4]}})
     False INT=&SelectIntFromArray(Direct Array Index=$Particles.PartitionParticles.Partition, Int Selection Array=DataInterface() {propertyValues {IntData [2,4,7,9]}})
    )
    Bone Sampling Mode=::NewEnumerator4
    MaskSpawnGroup=false
    Orientation Sampling=::NewEnumerator1
    Sampled Position Offset=&AddVectorToPosition(
     Position=&RandomRangeVector(Minimum=<10 10 10>, Maximum=<-10 -10 -10>)
     Vector=&SelectVectorFromArray(
      Direct Array Index=$Particles.PartitionParticles.Partition
      Vector Selection Array=DataInterface() {propertyValues {FloatData [<12 0 0>, <-12 0 0>, <0 0 0>, <0 0 0>]}}
     )
    )
    Skeletal Mesh=$System.Skeletal Mesh Reader
   )
   &AddVelocity(Module.Use Falloff Distance=false, Override Seed=false, Velocity Mode=::NewEnumerator1, Velocity Speed=10)
   &AddVelocity(Velocity=<0 0 1>) {enabled=false, module=AddVelocity001}
  ]
  particleUpdateScript [
   &ParticleState()
   &SpriteRotationRate(Rotation Rate=&RandomRangeFloat(Minimum=-80, Maximum=80))
   &AccelerationForce(Acceleration=&RandomRangeVector(Minimum=<0 0 20>, Maximum=<0 0 40>))
   &CurlNoiseForce(Bias Noise=false, Noise Strength=50, Pan Noise=false) {enabled=false}
   &SolveForcesAndVelocity(Clamp Velocity=false, Limit Acceleration=false)
   /SetVariables(Particles.ColorLight:LinearColor=&MakeLinearColorFromVectorAndFloat(
     "Vector (RGB)"=&Multiply_VectorByFloat(
      Vector=&MakeVectorFromLinearColorRGB(Color=&ColorFromCurve(ColorCurve=DataInterface() {propertyValues {
          AlphaCurve {keys []}
          BlueCurve {keys [<at 0.3427 0.3185 RCIM_Cubic>,<at 0.826 0.1875 RCIM_Cubic>]}
          GreenCurve {keys [<at 0.3427 0.5732 RCIM_Cubic>,<at 0.826 0.1542 RCIM_Cubic>]}
          RedCurve {keys [<at 0.3427 1 RCIM_Cubic>,<at 0.826 1.5 RCIM_Cubic>]}
         }}))
      Float=&Lerp_Float(
       A=1
       B=25
       Alpha=&Power(a=&Multiply_Float(A=$System.Audio Reactive, B=$NPC.FortniteNPC.EventAudio_Amplitude), b=10)
      )
     )
     "Float (Alpha)"=&FloatFromCurve(FloatCurve=DataInterface() {propertyValues {Curve {keys [<at 0 0>,<at 0.1 1>,<at 0.8 1>,<at 1 0>]}}})
    )) {isSetParameters=true}
   &ScaleColor(
    Scale Alpha=&FloatFromCurve(FloatCurve=DataInterface() {propertyValues {Curve {keys [<at 0 0>,<at 0.2 1>,<at 0.8 1>,<at 1 0>]}}})
    ScaleA=true
    ScaleRGB=true
   )
   /SetVariables(Particles.ColorGlow:LinearColor=&MakeLinearColorFromVectorAndFloat(
     "Vector (RGB)"=&MakeVectorFromLinearColorRGB(Color=<rgba 1 0.0677 0.0677 1>)
     "Float (Alpha)"=&FloatFromCurve(
      FloatCurve=DataInterface() {propertyValues {Curve {keys [<at 0 0>,<at 0.3 1>,<at 0.6 1>,<at 1 0>]}}}
      Scale Curve=&Lerp_Float(A=0, B=0.5, Alpha=$System.InterpolateOverTime.InterpolateOverTime_Float.Moving Average)
     )
    )) {isSetParameters=true}
   /SetVariables(Particles.SizeGlow:Vector2f=&Multiply_Vector2D_ByFloat(Float=1.6, Vector2D=$Particles.SpriteSize)) {isSetParameters=true}
   &ScaleSpriteSize(
    Uniform Curve Scale=&Multiply_Float(
     A=&RandomRangeFloat(Minimum=0.8, Maximum=1)
     B=&Lerp_Float(
      A=&Lerp_Float(A=0, B=1, Alpha=$System.Audio Reactive)
      B=8
      Alpha=&Power(
       a=&Multiply_Float(A=$System.Audio Reactive, B=&ClampFloat(Float=$NPC.FortniteNPC.EventAudio_Amplitude, Min=0.5, Max=0.8))
       b=8
      )
     )
    )
    Uniform Curve Sprite Scale=DataInterface() {propertyValues {Curve {keys [
        {tangentMode=RCTM_User, time=0, value=0}
        <at 0.2 0.65>
        <at 0.3 0.9>
        <at 0.5 1>
        <at 0.7 0.9>
        <at 0.9 1>
        {tangentMode=RCTM_User, time=1, value=0}
       ]}}}
   ) {module=ScaleSpriteSize001}
   &ScaleSpriteSize(
    Scale Sprite Size Mode=::NewEnumerator0
    Uniform Scale Factor=&MakeCustomFloatFromBool(
     Bool=$User.bisFootHidden
     True Float=&SelectFloatFromArray(Direct Array Index=$Particles.PartitionParticles.Partition, Float Selection Array=DataInterface() {propertyValues {FloatData [1,1,0,0]}})
     False Float=&SelectFloatFromArray(Direct Array Index=$Particles.PartitionParticles.Partition, Float Selection Array=DataInterface() {propertyValues {FloatData [1,1,1,1]}})
    )
   )
   &ScaleSpriteSize(
    Scale Sprite Size Mode=::NewEnumerator0
    Uniform Scale Factor=&MakeCustomFloatFromBool(
     Bool=$User.bisHandHidden
     True Float=&SelectFloatFromArray(Direct Array Index=$Particles.PartitionParticles.Partition, Float Selection Array=DataInterface() {propertyValues {FloatData [0,0,1,1]}})
     False Float=&SelectFloatFromArray(Direct Array Index=$Particles.PartitionParticles.Partition, Float Selection Array=DataInterface() {propertyValues {FloatData [1,1,1,1]}})
    )
   ) {enabled=false, module=ScaleSpriteSize002}
  ]
  rendererClasses [&NiagaraSpriteRendererProperties]
  renderers [
   {rendererClass=&NiagaraSpriteRendererProperties}
   {rendererClass=&NiagaraSpriteRendererProperties, rendererIndex=1}
   {rendererClass=&NiagaraSpriteRendererProperties, rendererIndex=2}
  ]
 }
 Star_Particle {
  emitterUpdateScript [
   &EmitterState(Inactive Response=::NewEnumerator1, Life Cycle Mode=::NewEnumerator1, Loop Duration=1, UseLoopDelay=false)
   &SpawnRate(SpawnRate=20, Use Spawn Probability=false) {enabled=false}
   &SpawnParticlesFromOtherEmitter(
    Attribute Reader=DataInterface() {propertyValues {EmitterBinding {bindingMode=Other, emitterName=Star}}}
    Spawn Rate=&Lerp_Float(A=0, B=20, Alpha=$System.Audio Reactive)
    Use Maximum Age=false
    Use Minimum Age=false
   )
  ]
  particleSpawnScript [
   &PartitionParticles(Number Of Partitions=4)
   &InitializeParticle(
    Color=<rgba 2 0 0.035 1>
    Lifetime Max=0.4
    Lifetime Min=0.3
    Lifetime Mode=::NewEnumerator1
    Sprite Size Mode=::NewEnumerator1
    Uniform Sprite Size Max=15
    Uniform Sprite Size Min=8
    UsePositionOffset=false
    Write Material Random=true
   )
   &SampleParticlesFromOtherEmitter(Applied Velocity Scale=0.05, Velocity Sampling=::NewEnumerator0)
   &ShapeLocation(MaskSpawnGroup=false, Override Seed=false, Sphere Radius=1, UseOwnerScale=true)
  ]
  particleUpdateScript [
   &ParticleState()
   &ScaleSpriteSize(
    Scale Sprite Size Mode=::NewEnumerator0
    Uniform Scale Factor=&MakeCustomFloatFromBool(
     Bool=$User.bisFootHidden
     True Float=&SelectFloatFromArray(Direct Array Index=$Particles.PartitionParticles.Partition, Float Selection Array=DataInterface() {propertyValues {FloatData [1,1,0,0]}})
     False Float=&SelectFloatFromArray(Direct Array Index=$Particles.PartitionParticles.Partition, Float Selection Array=DataInterface() {propertyValues {FloatData [1,1,1,1]}})
    )
   ) {module=ScaleSpriteSize001}
   &ScaleSpriteSize(
    Scale Sprite Size Mode=::NewEnumerator0
    Uniform Scale Factor=&MakeCustomFloatFromBool(
     Bool=$User.bisHandHidden
     True Float=&SelectFloatFromArray(Direct Array Index=$Particles.PartitionParticles.Partition, Float Selection Array=DataInterface() {propertyValues {FloatData [0,0,1,1]}})
     False Float=&SelectFloatFromArray(Direct Array Index=$Particles.PartitionParticles.Partition, Float Selection Array=DataInterface() {propertyValues {FloatData [1,1,1,1]}})
    )
   ) {enabled=false, module=ScaleSpriteSize002}
   &ScaleSpriteSize(
    Uniform Curve Scale=&RandomRangeFloat(Minimum=0.3, Maximum=1, Evaluation Type=::NewEnumerator1)
    Uniform Curve Sprite Scale=DataInterface() {propertyValues {Curve {keys [<at 0 0.6 RCIM_Cubic>, {arriveTangent=-0.2, interpMode=RCIM_Cubic, leaveTangent=-0.2, time=0.3, value=1}, <at 1 0.4 RCIM_Cubic>]}}}
   )
   &InheritSourceMovement(ApplyForce=false, ApplyPosition=true, ApplyPositionFalloff=false)
   &SolveForcesAndVelocity(Clamp Velocity=false, Limit Acceleration=false)
   &ScaleColor(
    Scale Alpha=&FloatFromCurve(FloatCurve=DataInterface() {propertyValues {Curve {keys [<at 0 0 RCIM_Cubic>,<at 0.1 1 RCIM_Cubic>,<at 1 0 RCIM_Cubic>]}}})
    ScaleA=true
    ScaleRGB=true
   )
  ]
  rendererClasses [&NiagaraSpriteRendererProperties]
  renderers [{rendererClass=&NiagaraSpriteRendererProperties}]
 }
}
summary {
 systemName=NS_Character_ChimeCurlCorn_Idle
 emitters [
  {emitterName=Pulse, rendererClasses [&NiagaraSpriteRendererProperties]}
  {emitterName=Star, rendererClasses [&NiagaraSpriteRendererProperties]}
  {emitterName=Ring_Limbs, rendererClasses [&NiagaraMeshRendererProperties]}
  {emitterName=Star_Particle, rendererClasses [&NiagaraSpriteRendererProperties]}
  {emitterName=Ring_Limbs_small, rendererClasses [&NiagaraMeshRendererProperties]}
  {emitterName=Light, rendererClasses [&NiagaraSpriteRendererProperties, &NiagaraLightRendererProperties]}
 ]
 userVariables [
  User.bisFootHidden {type=NiagaraBool}
  User.bIsFrontEndPreview {type=NiagaraBool}
  User.bisHandHidden {type=NiagaraBool}
 ]
}
