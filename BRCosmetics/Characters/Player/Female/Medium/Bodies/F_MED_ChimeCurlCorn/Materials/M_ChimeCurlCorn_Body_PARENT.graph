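# Parent body material for the ChimeCurlCorn female character cosmetic, adding animated sequin sparkle and optional audio reactivity between the 2023 character parent PreFX and PostFX functions.
# Rotated, tiled sequin texture values offset a time-scrolled noise lookup, multiplied by a panning gradient and tint to form sparkle, blended into emissive via a Tech Art Mask channel.
# A default-on static switch feeds frontend-loop audio amplitude, masked by M's blue channel, into MF_ChimeCurlCorn.
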Material /BRCosmetics/Characters/Player/Female/Medium/Bodies/F_MED_ChimeCurlCorn/Materials/M_ChimeCurlCorn_Body_PARENT
graph {
 properties {
  bUseMaterialAttributes=true
  bUsedWithMeshDeformer=true
  bUsedWithMorphTargets=true
  bUsedWithSkeletalMesh=true
  pixelDepthOffsetMode=PDOM_Legacy
  shadingModel=MSM_FromMaterialExpression
  usedShadingModels=Default Lit | Subsurface | Clear Coat | Subsurface Profile | Cloth
 }
 nodes [
  n1 = MaterialFunctionCall() {materialFunction=&MF_CharacterParent_2023_PreFX}
  n2 = GetMaterialAttributes(None=@n1.Result) {attributeGetTypes [#1]}
  n3 = SetMaterialAttributes(
   MaterialAttributes=@n2.MaterialAttributes
   Emissive Color=LinearInterpolate(
    A=@n2.Emissive Color
    B=NamedRerouteUsage()
    Alpha=ChannelMaskParameter(None=@n1.Tech Art Mask) {group=Sequin, parameterName=SequinChannel, sortPriority=0}.R
   )
  ) {attributeSetTypes [#1]}
  n4 = Power(Base=MaterialFunctionCall!(InAmplitude, InSpectrum, CustomFakeSpectrum, CustomFakeAmplitude) {materialFunction=/BRCosmetics/Global/MaterialFunctions/AudioReactivity/MF_AudioReactivity_FrontendLoop}.Amplitude) {constExponent=1.5}
  n5 = Multiply(A=Saturate(None=Multiply(A=@n4) {constB=4}), B=Power(Base=ComponentMask(None=@n1.M) {b=true}) {constExponent=4}) {constA=1}
  MaterialFunctionCall(
   Material=StaticSwitchParameter(
    True=MaterialFunctionCall!(Material=@n3, TechArtMask=@n5) {materialFunction=&/MF_ChimeCurlCorn}.Result
    False=@n3
   ) {defaultValue=true, parameterName=UseAudioReactivity}
   M=@n1.M
   Pre Skinned Local Position=@n1.M
   Distance Blend=@n1.M
   Fresnel=@n1.M
  ) {materialFunction=&MF_CharacterParent_2023_PostFX}
  n6 = TextureCoordinate()
  n7 = Constant()
  n8 = ScalarParameter() {group=Sequin, parameterName=SequinRotationAngle}
  NamedRerouteDeclaration(None=Multiply(
    A=MaterialFunctionCall(UVs=@n6, Rotation Center=@n7, "Rotation Angle (0-1)"=@n8) {materialFunction=&CustomRotator}.Rotated Values
    B=ScalarParameter() {defaultValue=32, group=Sequin, parameterName=SequinTile, sortPriority=5}
   )) {name=SequinUVs, nodeColor=<rgba 0 1 0.0859 1>}
  n9 = TextureObjectParameter() {
   group=Sequin
   parameterName=SequinOffest
   samplerType=SAMPLERTYPE_LinearGrayscale
   sortPriority=1
   texture=/Game/Global/Textures/Default/Tileable_Patterns/T_SequinTile
  }
  n10 = TextureSample!(UVs=NamedRerouteUsage(), Tex=@n9, Bias=ScalarParameter() {group=Sequin, parameterName=SequinTextureMipBias}, Apply View MipBias) {
   mipValueMode=TMVM_MipBias
   samplerSource=SSM_Wrap_WorldGroupSettings
   samplerType=SAMPLERTYPE_LinearGrayscale
   texture=&T_SequinTile
  }
  n11 = Time()
  n12 = ScalarParameter() {defaultValue=0.01, group=Sequin, parameterName=SequinSparkleSpeed, sortPriority=12}
  n13 = TextureSample!(UVs=Add(A=ComponentMask(None=@n10.RGB) {g=true, r=true}, B=Multiply(A=@n11, B=@n12) {constB=0.2}), Tex, Apply View MipBias) {samplerSource=SSM_Wrap_WorldGroupSettings, samplerType=SAMPLERTYPE_Normal, texture=/Game/Global/Textures/Default/Noise/64x64Noise}
  n14 = TextureCoordinate()
  n15 = ScalarParameter() {defaultValue=2, group=Sequin, parameterName=SequinEmissiveTiling, sortPriority=14}
  n16 = VectorParameter() {group=Sequin, parameterName=SequinEmissivePanSpeed, sortPriority=14, defaultValue=<rgba 0.05 0.05 0 0>}
  n17 = TextureSampleParameter2D(UVs=Panner(Coordinate=Multiply(A=@n14, B=@n15), Speed=ComponentMask(None=@n16.RGB) {g=true, r=true}) {speedX=0.05, speedY=0.05}) {
   group=Sequin
   parameterName=SequinEmissiveGradient
   samplerSource=SSM_Wrap_WorldGroupSettings
   sortPriority=14
   texture=/Game/Effects/Fort_Effects/Textures/Tile/T_ShieldCaustics_Tile_Chromatic
  }
  n18 = Multiply(
   A=Power(Base=@n13.R) {constExponent=4}
   B=Power(Base=ChannelMaskParameter(None=@n17.RGB) {group=Sequin, parameterName=SequinEmissiveGradientChannel, sortPriority=14}.R)
  )
  n19 = VectorParameter() {group=Sequin, parameterName=SparkleColor, defaultValue=<rgba 0.5742 0.228 1 1>}
  NamedRerouteDeclaration(None=Multiply(A=Multiply(A=@n18, B=@n19.RGB), B=ScalarParameter() {defaultValue=10, group=Sequin, parameterName=SparkleBRightness, sortPriority=13})) {name=Emissive_Out, nodeColor=<rgba 0.0312 1 0 1>}
 ]
}
