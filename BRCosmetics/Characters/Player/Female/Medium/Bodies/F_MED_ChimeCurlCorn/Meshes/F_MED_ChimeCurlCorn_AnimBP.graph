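# Anim Blueprint for the F_MED_ChimeCurlCorn female player body, adding facial pose, Control Rig and zipper secondary motion.
# It copies the parent pose, layers the head part's pose from FACIAL_C_FacialRoot, then runs the body's Control Rig.
# Two spring-limited AnimDynamics nodes simulate the left and right zipper bones.
# Each update stores a movement-state gravity vector, but the dynamics nodes use fixed gravity and no RigidBody node consumes the RBAN parameters.
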AnimBlueprint /BRCosmetics/Characters/Player/Female/Medium/Bodies/F_MED_ChimeCurlCorn/Meshes/F_MED_ChimeCurlCorn_AnimBP
target_skeleton=&/F_MED_ChimeCurlCorn_Skeleton
graphs [
 AnimGraph {nodes [
   n1 = Get head() -> (head:Skeletal Mesh Component Object Reference)
   n2 = Control Rig(Source=Layered blend per bone(
     BasePose=Copy Pose From Mesh() {bUseAttachedParent=true}
     BlendPoses_0=Copy Pose From Mesh(SourceMeshComponent=@n1) {bUseAttachedParent=true}
    ) {layerSetup [{branchFilters [{boneName=FACIAL_C_FacialRoot}]}]}) {
    alphaBoolBlend {blendOption=Linear}
    controlRigAssetReference {blueprintRigClass=&/F_MED_ChimeCurlCorn_CtrlRig.F_MED_ChimeCurlCorn_CtrlRig_C}
   }
   n3 = Anim Dynamics - Bone: dyn_zipper_l!(ComponentPose=Local To Component(LocalPose=@n2), Alpha=1, GravityOverride:Vector="0,0,-980", ExternalForce="0,0,0") -> (Pose) {
    angularDampingOverride=0.9999
    angularSpringConstant=8000
    bAngularSpring=true
    bOverrideAngularDamping=true
    bUseGravityOverride=true
    bUsePlanarLimit=false
    gravityScale=0.35
    simulationSpace=RootRelative
    alphaBoolBlend {blendOption=Linear}
    boundBone {boneName=dyn_zipper_l}
    componentAppliedLinearAccClamp=<1400 1400 1400>
    componentLinearAccScale=<1 1 1>
    componentLinearVelScale=<1 1 1>
    gravityOverride {z=-980}
    physicsBodyDefinitions [{
      boundBone {boneName=dyn_zipper_l}
      boxExtents=<20 15 10>
      constraintSetup {angularLimitsMax=<15 45 90>, angularLimitsMin=<-15 0 -90>, angularTarget=<1 0 0>}
      localJointOffset=<4 0 1>
     }]
   }
   n4 = Anim Dynamics - Bone: dyn_zipper_r!(ComponentPose=@n3, Alpha=1, GravityOverride:Vector="0,0,-980", ExternalForce="0,0,0") -> (Pose) {
    angularDampingOverride=0.9999
    angularSpringConstant=8000
    bAngularSpring=true
    bOverrideAngularDamping=true
    bUseGravityOverride=true
    bUsePlanarLimit=false
    gravityScale=0.35
    simulationSpace=RootRelative
    alphaBoolBlend {blendOption=Linear}
    boundBone {boneName=dyn_zipper_r}
    componentAppliedLinearAccClamp=<1400 1400 1400>
    componentLinearAccScale=<1 1 1>
    componentLinearVelScale=<1 1 1>
    gravityOverride {z=-980}
    physicsBodyDefinitions [{
      boundBone {boneName=dyn_zipper_r}
      boxExtents=<20 15 10>
      constraintSetup {angularLimitsMax=<15 45 90>, angularLimitsMin=<-15 0 -90>, angularTarget=<1 0 0>}
      localJointOffset=<4 0 1>
     }]
   }
   Output Pose(Result=Component To Local(ComponentPose=@n4)) {layerGroup=None}
  ]}
 EventGraph {nodes [
   Event BlueprintUpdateAnimation()
   > n1 = "Gravity Override State Function (Nativized)"!(
    execute
    CharacterPartAnimInstance:Custom Character Part Anim Instance Object Reference=Self-Reference()
    GravityOverrideParameterSet:Fort Gravity Override Parameter Set Structure=Get Gravity Override Params() -> (Gravity Override Params:Fort Gravity Override Parameter Set Structure)
    RigidBodyAnimNodeParameterSet:Fort Rigid Body Anim Node Parameter Set Structure=Get RBAN Params() -> (RBAN Params:Fort Rigid Body Anim Node Parameter Set Structure)
    DebugDraw:bool=false
    GameTimeInSeconds:float=GetGameTimeInSeconds()
    DebugJointName:Name=root
    StateLogicLayerName:Name=none
   ) -> (
    then
    OutGravityOverride:Vector
    OutCurrentState:Name=None
    OutGravityOverrideParameters:Fort Gravity Override Parameters Structure
    OutRigidBodyAnimNodeParameters:Fort Rigid Body Anim Node Parameters Structure
   )
   > Set Gravity Override(Gravity Override=@n1)
   Event OnCharacterCustomizationLoaded()
   > n2 = Cast To PlayerPawn_Athena(Object=TryGetPawnOwner())
   > Set head(head:Skeletal Mesh Component Object Reference=GetSkeletalMeshForPartType(self=@n2, PartType=Head)) -> (Output_Get:Skeletal Mesh Component Object Reference)
  ]}
]
variables [
 Gravity Override {type=Vector}
 Gravity Override Params {
  type=FortGravityOverrideParameterSet
  default=(Skydive_GoingUp=(GravityOverride=(WindFrequency=(X=1.500000,Y=7.000000,Z=8.000000),WindAmplitude=(X=200.000000,Y=-500.000000,Z=400.000000),GravityOverride=(X=0.000000,Y=-300.000000,Z=100.000000),JointName="root")),Skydive_FloatingDown=(GravityOverride=(WindFrequency=(X=1.500000,Y=7.000000,Z=8.000000),WindAmplitude=(X=200.000000,Y=-500.000000,Z=400.000000),WindOffset=(X=0.000000,Y=-600.000000,Z=400.000000),GravityOverride=(X=0.000000,Y=-700.000000,Z=600.000000),JointName="root")),Skydive_DivingDown=(GravityOverride=(WindFrequency=(X=1.500000,Y=12.000000,Z=10.000000),WindAmplitude=(X=50.000000,Y=300.000000,Z=100.000000),GravityOverride=(X=-1500.000000,Y=0.000000,Z=0.000000),JointName="pelvis")),Skydive_Parachuting=(GravityOverride=(WindFrequency=(X=1.500000,Y=7.000000,Z=8.000000),WindAmplitude=(X=200.000000,Y=-500.000000,Z=400.000000),GravityOverride=(X=0.000000,Y=-800.000000,Z=-980.000000),JointName="root")),OnGround_Standing=(GravityOverride=(GravityOverride=(X=0.000000,Y=0.000000,Z=-980.000000),JointName="root")),OnGround_Moving=(GravityOverride=(GravityOverride=(X=0.000000,Y=0.000000,Z=-980.000000),JointName="root")),OnGround_Moving_HighRotationRate=(GravityOverride=(GravityOverride=(X=0.000000,Y=0.000000,Z=-980.000000),JointName="root")),OnGround_Crouch_Moving=(GravityOverride=(GravityOverride=(X=0.000000,Y=0.000000,Z=-980.000000),JointName="root")),Emote_or_Melee=(GravityOverride=(GravityOverride=(X=0.000000,Y=0.000000,Z=-980.000000),JointName="root")),Swimming=(GravityOverride=(GravityOverride=(X=-800.000000,Y=0.000000,Z=0.000000),JointName="pelvis")),Front_End=(GravityOverride=(GravityOverride=(X=0.000000,Y=0.000000,Z=-980.000000),JointName="root")),DBNO=(GravityOverride=(GravityOverride=(X=-980.000000,Y=400.000000,Z=0.000000),JointName="pelvis")),DBNO_Carried=(GravityOverride=(GravityOverride=(X=-980.000000,Y=500.000000,Z=-500.000000),JointName="pelvis")),OnGround_Falling=(GravityOverride=(GravityOverride=(X=0.000000,Y=0.000000,Z=-980.000000),JointName="root")),Ziplining=(GravityOverride=(GravityOverride=(X=0.000000,Y=0.000000,Z=-980.000000),JointName="root")),Skydive_Surfing=(GravityOverride=(WindFrequency=(X=1.500000,Y=7.000000,Z=8.000000),WindAmplitude=(X=200.000000,Y=-500.000000,Z=400.000000),GravityOverride=(X=500.000000,Y=-300.000000,Z=-980.000000),JointName="root")),Riding_Standing=(GravityOverride=(GravityOverride=(X=0.000000,Y=-350.000000,Z=-980.000000),JointName="root")),Riding_Moving=(GravityOverride=(GravityOverride=(X=0.000000,Y=-450.000000,Z=-450.000000),JointName="root")),Riding_Moving_Targeting=(GravityOverride=(GravityOverride=(X=0.000000,Y=-200.000000,Z=-980.000000),JointName="root")),Tactical_Sprint=(GravityOverride=(GravityOverride=(X=0.000000,Y=-450.000000,Z=-500.000000),JointName="root")),Rail_Grinding=(GravityOverride=(GravityOverride=(X=0.000000,Y=0.000000,Z=-980.000000),JointName="root")),)
 }
 head {type=SkeletalMeshComponent}
 RBAN Params {
  type=FortRigidBodyAnimNodeParameterSet
  default=(Skydive_GoingUp=(RigidBodyAnimNode=(ComponentLinearAccScale=(X=0.000000,Y=-0.500000,Z=0.000000),ComponentLinearVelScale=(X=0.000000,Y=-0.500000,Z=0.000000),SimSpaceSettings=(WorldAlpha=1.000000,VelocityScaleZ=0.000000,DampingAlpha=0.000000,MaxLinearVelocity=300.000000,MaxAngularVelocity=4.000000,MaxLinearAcceleration=1200.000000,MaxAngularAcceleration=16.000000,ExternalLinearDragV=(X=0.000000,Y=0.000000,Z=-1.000000)))),Skydive_FloatingDown=(RigidBodyAnimNode=(ComponentLinearAccScale=(X=0.000000,Y=0.500000,Z=0.000000),ComponentLinearVelScale=(X=0.000000,Y=0.500000,Z=0.000000),SimSpaceSettings=(WorldAlpha=1.000000,VelocityScaleZ=0.500000,DampingAlpha=0.000000,MaxLinearVelocity=700.000000,MaxAngularVelocity=8.000000,MaxLinearAcceleration=2800.000000,MaxAngularAcceleration=32.000000,ExternalLinearDragV=(X=0.500000,Y=0.500000,Z=0.500000)))),Skydive_DivingDown=(RigidBodyAnimNode=(ComponentLinearAccScale=(X=0.100000,Y=0.000000,Z=0.000000),ComponentLinearVelScale=(X=0.100000,Y=0.000000,Z=0.000000),SimSpaceSettings=(WorldAlpha=1.000000,VelocityScaleZ=0.500000,DampingAlpha=0.000000,MaxLinearVelocity=700.000000,MaxAngularVelocity=8.000000,MaxLinearAcceleration=2800.000000,MaxAngularAcceleration=32.000000,ExternalLinearDragV=(X=0.500000,Y=0.500000,Z=0.500000)))),Skydive_Parachuting=(RigidBodyAnimNode=(ComponentLinearAccScale=(X=0.000000,Y=0.100000,Z=0.300000),ComponentLinearVelScale=(X=0.000000,Y=0.100000,Z=0.300000),SimSpaceSettings=(WorldAlpha=1.000000,VelocityScaleZ=0.500000,DampingAlpha=0.000000,MaxLinearVelocity=700.000000,MaxAngularVelocity=8.000000,MaxLinearAcceleration=2800.000000,MaxAngularAcceleration=32.000000,ExternalLinearDragV=(X=0.500000,Y=0.500000,Z=0.500000)))),OnGround_Standing=(RigidBodyAnimNode=(ComponentLinearAccScale=(X=0.300000,Y=0.000000,Z=0.000000),ComponentLinearVelScale=(X=0.300000,Y=0.000000,Z=0.000000),SimSpaceSettings=(WorldAlpha=1.000000,VelocityScaleZ=0.500000,DampingAlpha=0.000000,MaxLinearVelocity=500.000000,MaxAngularVelocity=6.000000,MaxLinearAcceleration=2000.000000,MaxAngularAcceleration=24.000000,ExternalLinearDragV=(X=0.200000,Y=0.200000,Z=0.200000)))),OnGround_Moving=(RigidBodyAnimNode=(ComponentLinearAccScale=(X=0.300000,Y=0.000000,Z=0.000000),ComponentLinearVelScale=(X=0.300000,Y=0.000000,Z=0.000000),SimSpaceSettings=(WorldAlpha=1.000000,VelocityScaleZ=0.500000,DampingAlpha=0.000000,MaxLinearVelocity=500.000000,MaxAngularVelocity=6.000000,MaxLinearAcceleration=2000.000000,MaxAngularAcceleration=24.000000,ExternalLinearDragV=(X=0.200000,Y=0.200000,Z=0.200000)))),OnGround_Moving_HighRotationRate=(RigidBodyAnimNode=(SimSpaceSettings=(WorldAlpha=1.000000,VelocityScaleZ=0.500000,DampingAlpha=0.000000,MaxLinearVelocity=500.000000,MaxAngularVelocity=6.000000,MaxLinearAcceleration=2000.000000,MaxAngularAcceleration=24.000000,ExternalLinearDragV=(X=0.200000,Y=0.200000,Z=0.200000)))),OnGround_Crouch_Moving=(RigidBodyAnimNode=(SimSpaceSettings=(WorldAlpha=1.000000,VelocityScaleZ=0.500000,DampingAlpha=0.000000,MaxLinearVelocity=400.000000,MaxAngularVelocity=8.000000,MaxLinearAcceleration=3200.000000,MaxAngularAcceleration=32.000000,ExternalLinearDragV=(X=0.500000,Y=0.500000,Z=0.500000)))),Emote_or_Melee=(RigidBodyAnimNode=(SimSpaceSettings=(WorldAlpha=1.000000,VelocityScaleZ=0.500000,DampingAlpha=0.000000,MaxLinearVelocity=700.000000,MaxAngularVelocity=8.000000,MaxLinearAcceleration=2800.000000,MaxAngularAcceleration=32.000000,ExternalLinearDragV=(X=0.500000,Y=0.500000,Z=0.500000)))),Swimming=(RigidBodyAnimNode=(SimSpaceSettings=(WorldAlpha=1.000000,VelocityScaleZ=0.500000,DampingAlpha=0.000000,MaxLinearVelocity=700.000000,MaxAngularVelocity=8.000000,MaxLinearAcceleration=2800.000000,MaxAngularAcceleration=32.000000,ExternalLinearDragV=(X=0.500000,Y=0.500000,Z=0.500000)))),Front_End=(RigidBodyAnimNode=(SimSpaceSettings=(WorldAlpha=1.000000,VelocityScaleZ=0.500000,DampingAlpha=0.000000,MaxLinearVelocity=200.000000,MaxAngularVelocity=2.500000,MaxLinearAcceleration=800.000000,MaxAngularAcceleration=10.000000))),DBNO=(RigidBodyAnimNode=(SimSpaceSettings=(WorldAlpha=1.000000,VelocityScaleZ=0.500000,DampingAlpha=0.000000,MaxLinearVelocity=700.000000,MaxAngularVelocity=8.000000,MaxLinearAcceleration=2800.000000,MaxAngularAcceleration=32.000000,ExternalLinearDragV=(X=0.500000,Y=0.500000,Z=0.500000)))),DBNO_Carried=(RigidBodyAnimNode=(SimSpaceSettings=(WorldAlpha=1.000000,VelocityScaleZ=0.500000,DampingAlpha=0.000000,MaxLinearVelocity=700.000000,MaxAngularVelocity=8.000000,MaxLinearAcceleration=2800.000000,MaxAngularAcceleration=32.000000,ExternalLinearDragV=(X=0.500000,Y=0.500000,Z=0.500000)))),OnGround_Falling=(RigidBodyAnimNode=(SimSpaceSettings=(WorldAlpha=1.000000,VelocityScaleZ=0.300000,DampingAlpha=0.000000,MaxLinearVelocity=700.000000,MaxAngularVelocity=8.000000,MaxLinearAcceleration=2800.000000,MaxAngularAcceleration=32.000000))),Ziplining=(RigidBodyAnimNode=(SimSpaceSettings=(WorldAlpha=1.000000,VelocityScaleZ=0.500000,DampingAlpha=0.000000,MaxLinearVelocity=700.000000,MaxAngularVelocity=8.000000,MaxLinearAcceleration=2800.000000,MaxAngularAcceleration=32.000000,ExternalLinearDragV=(X=0.500000,Y=0.500000,Z=0.500000)))),Skydive_Surfing=(RigidBodyAnimNode=(SimSpaceSettings=(WorldAlpha=1.000000,VelocityScaleZ=0.500000,DampingAlpha=0.000000,MaxLinearVelocity=700.000000,MaxAngularVelocity=8.000000,MaxLinearAcceleration=2800.000000,MaxAngularAcceleration=32.000000,ExternalLinearDragV=(X=0.500000,Y=0.500000,Z=0.500000)))),Riding_Standing=(RigidBodyAnimNode=(SimSpaceSettings=(WorldAlpha=1.000000,VelocityScaleZ=0.500000,DampingAlpha=0.000000,MaxLinearVelocity=700.000000,MaxAngularVelocity=8.000000,MaxLinearAcceleration=2800.000000,MaxAngularAcceleration=32.000000,ExternalLinearDragV=(X=0.500000,Y=0.500000,Z=0.500000)))),Riding_Moving=(RigidBodyAnimNode=(SimSpaceSettings=(WorldAlpha=1.000000,VelocityScaleZ=0.500000,DampingAlpha=0.000000,MaxLinearVelocity=700.000000,MaxAngularVelocity=8.000000,MaxLinearAcceleration=2800.000000,MaxAngularAcceleration=32.000000,ExternalLinearDragV=(X=0.500000,Y=0.500000,Z=0.500000)))),Riding_Moving_Targeting=(RigidBodyAnimNode=(SimSpaceSettings=(WorldAlpha=1.000000,VelocityScaleZ=0.500000,DampingAlpha=0.000000,MaxLinearVelocity=700.000000,MaxAngularVelocity=8.000000,MaxLinearAcceleration=2800.000000,MaxAngularAcceleration=32.000000,ExternalLinearDragV=(X=0.500000,Y=0.500000,Z=0.500000)))),Tactical_Sprint=(RigidBodyAnimNode=(ComponentLinearAccScale=(X=0.300000,Y=0.000000,Z=0.000000),ComponentLinearVelScale=(X=0.300000,Y=0.000000,Z=0.000000),SimSpaceSettings=(WorldAlpha=1.000000,VelocityScaleZ=0.250000,DampingAlpha=0.000000,MaxLinearVelocity=500.000000,MaxAngularVelocity=6.000000,MaxLinearAcceleration=2000.000000,MaxAngularAcceleration=24.000000,ExternalLinearDragV=(X=0.200000,Y=0.200000,Z=0.200000)))),Rail_Grinding=(RigidBodyAnimNode=(SimSpaceSettings=(WorldAlpha=1.000000,VelocityScaleZ=0.500000,DampingAlpha=0.000000,MaxLinearVelocity=700.000000,MaxAngularVelocity=8.000000,MaxLinearAcceleration=2800.000000,MaxAngularAcceleration=32.000000,ExternalLinearDragV=(X=0.500000,Y=0.500000,Z=0.500000)))),)
 }
]
