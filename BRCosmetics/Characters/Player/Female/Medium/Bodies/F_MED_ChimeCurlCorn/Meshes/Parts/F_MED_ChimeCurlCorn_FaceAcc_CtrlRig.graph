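# Control Rig for the ChimeCurlCorn female character's face accessory part, procedurally posing the ponytail bones in the forward solve.
# Two Aim units point the first two ponytail bones at the top spline guide bone.
# A Hermite spline is then built through four guide bones parented to the neck and upper spine, and the remaining four ponytail bones are fitted along it with stretched alignment.
# Side-bang bones exist in the hierarchy but are not driven by the graph.
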ControlRigBlueprint /BRCosmetics/Characters/Player/Female/Medium/Bodies/F_MED_ChimeCurlCorn/Meshes/Parts/F_MED_ChimeCurlCorn_FaceAcc_CtrlRig
graphs [
 RigVMFunctionLibrary {links [], nodes []}
 RigVMModel {nodes [
   RigUnit_BeginExecution = &RigUnit_BeginExecution()
   /Script/ControlRig.RigUnit_AimItem(
    DebugSettings=(bEnabled=False,Scale=10,WorldOffset=(Rotation=(X=0,Y=0,Z=0,W=1),Translation=(X=0,Y=0,Z=0),Scale3D=(X=1,Y=1,Z=1)))
    Item=(Type=Bone,Name="dyn_ponytail_01")
    Primary=(Weight=1,Axis=(X=1,Y=0,Z=0),Target=(X=0,Y=0,Z=-1),Kind=Location,Space=(Type=Bone,Name="dyn_ponytail_spline_00"))
    Secondary=(Weight=0,Axis=(X=0,Y=0,Z=1),Target=(X=1,Y=0,Z=0),Kind=Location,Space=(Type=Bone,Name="None"))
    Weight=1
    ExecutePin=@RigUnit_BeginExecution.ExecutePin
   ) {is=struct RigVMUnitNode}
   > &RigUnit_AimItem(
    DebugSettings=(bEnabled=False,Scale=10,WorldOffset=(Rotation=(X=0,Y=0,Z=0,W=1),Translation=(X=0,Y=0,Z=0),Scale3D=(X=1,Y=1,Z=1)))
    Item=(Type=Bone,Name="dyn_ponytail_02")
    Primary=(Weight=1,Axis=(X=1,Y=0,Z=0),Target=(X=0,Y=0,Z=-1),Kind=Location,Space=(Type=Bone,Name="dyn_ponytail_spline_00"))
    Secondary=(Weight=0,Axis=(X=0,Y=0,Z=1),Target=(X=1,Y=0,Z=0),Kind=Location,Space=(Type=Bone,Name="None"))
    Weight=1
   )
   > SplineFromItems = RigVMFunctionReferenceNode(
    Items=((Type=Bone,Name="dyn_ponytail_spline_00"),(Type=Bone,Name="dyn_ponytail_spline_01"),(Type=Bone,Name="dyn_ponytail_spline_02"),(Type=Bone,Name="dyn_ponytail_spline_03"))
    Samples Per Segment=16
    Spline Mode=Hermite
   )
   Make_Array = RigVMDispatchNode(Values=((Type=Bone,Name="dyn_ponytail_03"),(Type=Bone,Name="dyn_ponytail_04"),((),Type=Bone,Name="dyn_ponytail_05"),((),Type=Bone,Name="dyn_ponytail_06"))) {template="DISPATCH_RigVMDispatch_ArrayMake(in Values,out Array)"}
   > /Script/ControlRigSpline.RigUnit_FitChainToSplineCurveItemArray(
    Alignment=Stretched
    DebugSettings=(bEnabled=False,Scale=1,CurveColor=(R=1,G=1,B=0,A=1),SegmentsColor=(R=1,G=0,B=0,A=1),WorldOffset=(Rotation=(X=0,Y=0,Z=0,W=1),Translation=(X=0,Y=0,Z=0),Scale3D=(X=1,Y=1,Z=1)))
    Maximum=1
    Minimum=0
    PoleVectorPosition=(X=0,Y=0,Z=0)
    PrimaryAxis=(X=1,Y=0,Z=0)
    RotationEaseType=Linear
    Rotations=()
    SamplingPrecision=12
    SecondaryAxis=(X=0,Y=0,Z=0)
    Weight=1
    bPropagateToChildren=True
    Items=@Make_Array.Array
    Spline=@SplineFromItems.Spline
   ) {is=struct RigVMUnitNode}
  ]}
]
hierarchy [
 root {loc=<0 0 0>, rot=<0 0 0 1>}
 pelvis {parent=root, loc=<0 -0.5301 98.0571>, rot=<0.0127 0.707 -0.0127 -0.707>}
 spine_01 {parent=pelvis, loc=<6.1208 -0.9328 0>, rot=<0 0 0.034 -0.9994>}
 spine_02 {parent=spine_01, loc=<6.5997 -0.2095 0>, rot=<0 0 0.0238 -0.9997>}
 spine_03 {parent=spine_02, loc=<6.4914 -0.1237 0>, rot=<0 0 0.0149 -0.9999>}
 spine_04 {parent=spine_03, loc=<6.9598 0.0794 0>, rot=<0 0 -0.0132 -0.9999>}
 spine_05 {parent=spine_04, loc=<6.9987 0 0>, rot=<0 0 -0.0015 -1>}
 neck_01 {parent=spine_05, loc=<11.3138 0.1529 0>, rot=<0 0 -0.2003 -0.9797>}
 neck_02 {parent=neck_01, loc=<4.5627 0 0>, rot=<0 0 0.0504 -0.9987>}
 head {parent=neck_02, loc=<5.3914 0 0>, rot=<0 0 0.115 -0.9934>}
 dyn_ponytail_base {parent=head, loc=<13.9293 -8.4288 0.1654>, rot=<0.5489 -0.835 0.0212 -0.0328>}
 dyn_ponytail_01 {parent=dyn_ponytail_base, loc=<0 0 0>, rot=<0 0 0 1>}
 dyn_ponytail_02 {parent=dyn_ponytail_01, loc=<6.3617 0 0>, rot=<0.0277 -0.0296 -0.3922 -0.919>}
 dyn_ponytail_03 {parent=dyn_ponytail_02, loc=<8.6944 0 0>, rot=<-0.0003 0.0015 -0.1618 -0.9868>}
 dyn_ponytail_04 {parent=dyn_ponytail_03, loc=<8.4055 0 0>, rot=<0 -0.0006 -0.096 -0.9954>}
 dyn_ponytail_05 {parent=dyn_ponytail_04, loc=<7.4395 0 0>, rot=<0.0001 0.0008 -0.0075 -1>}
 dyn_ponytail_06 {parent=dyn_ponytail_05, loc=<6.5969 0 0>, rot=<-0.0002 -0.002 0.0946 -0.9955>}
 dyn_sidebangs_base_l {parent=head, loc=<7.2481 5.402 -6.0501>, rot=<0.0579 -0.9967 0.0028 -0.0562>}
 dyn_sidebangs_01_l {parent=dyn_sidebangs_base_l, loc=<0 0 0>, rot=<0 0 0 1>}
 dyn_sidebangs_02_l {parent=dyn_sidebangs_01_l, loc=<4.6161 0 0>, rot=<0.004 -0.0701 -0.05 -0.9963>}
 dyn_sidebangs_03_l {parent=dyn_sidebangs_02_l, loc=<4.6707 0 0>, rot=<-0.0061 -0.0867 -0.0778 -0.9932>}
 dyn_sidebangs_base_r {parent=head, loc=<7.2484 5.402 6.0501>, rot=<0.9967 0.0579 -0.0562 -0.0028>}
 dyn_sidebangs_01_r {parent=dyn_sidebangs_base_r, loc=<0 0 0>, rot=<0 0 0 -1>}
 dyn_sidebangs_02_r {parent=dyn_sidebangs_01_r, loc=<-4.6168 -0.0001 0.0001>, rot=<0.004 -0.0701 -0.05 -0.9963>}
 dyn_sidebangs_03_r {parent=dyn_sidebangs_02_r, loc=<-4.67 0 0>, rot=<-0.0061 -0.0867 -0.0778 -0.9932>}
 dyn_ponytail_spline_00 {parent=neck_02, loc=<0 -11.3724 0>, rot=<0 0 0 1>}
 dyn_ponytail_spline_01 {parent=neck_01, loc=<-1.4663 -10.082 0>, rot=<0 0 0 1>}
 clavicle_l {parent=spine_05, loc=<5.4831 -0.1053 -3.2946>, rot=<0.0465 -0.7289 -0.009 -0.6829>}
 clavicle_r {parent=spine_05, loc=<5.4831 -0.1053 3.2946>, rot=<0.7289 0.0465 -0.6829 0.009>}
 dyn_ponytail_spline_02 {parent=spine_05, loc=<5.5405 -9.9478 0>, rot=<0 0 0 -1>}
 dyn_ponytail_spline_03 {parent=spine_05, loc=<-1.1675 -10.6155 0>, rot=<0 0 0 1>}
]
