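# Physics asset for the ChimeCurlCorn female medium character's FaceAcc mesh, set up for a dynamic ponytail.
# A tapering chain of six capsules hangs from a small base sphere on the head, with asymmetric swing limits and zero twist between links.
# Sphere and capsule bodies on the head, neck, clavicles, upper spine and cape-collision bone form the surrounding body volume, joined by free angular constraints.
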PhysicsAsset /BRCosmetics/Characters/Player/Female/Medium/Bodies/F_MED_ChimeCurlCorn/Meshes/Parts/F_MED_ChimeCurlCorn_FaceAcc_Physics
bodies {
 clavicle_l [{shapeType=Capsule, length=18.977, radius=11.023, shapeName=clavicle_l_capsule, center=<17.3785 2.0619 -4.0473>, rotation=<rot 0 -90 90>}]
 clavicle_r [{shapeType=Capsule, length=19.2201, radius=10.7799, shapeName=clavicle_r_capsule, center=<-17.3875 -1.6674 3.824>, rotation=<rot 0 -90 90>}]
 dyn_ponytail_01 [{
   shapeType=Capsule
   length=2.7384
   radius=4.581
   shapeName=dyn_braid_01_capsule
   center=<2.3909 -1.8802 0.0482>
   rotation=<rot -1.3783 -72.9799 -88.2792>
  }]
 dyn_ponytail_02 [{
   shapeType=Capsule
   length=0
   radius=4.6664
   shapeName=dyn_braid_01_capsule
   center=<3.9047 -0.1843 0.0181>
   rotation=<rot -0.7653 -91.3584 -87.9325>
  }]
 dyn_ponytail_03 [{
   shapeType=Capsule
   length=0
   radius=4.1664
   shapeName=dyn_braid_01_capsule
   center=<2.6506 -0.021 0.0653>
   rotation=<rot -0.5746 -96.5612 -87.8716>
  }]
 dyn_ponytail_04 [{
   shapeType=Capsule
   length=1.7956
   radius=3.1664
   shapeName=dyn_braid_01_capsule
   center=<0.8218 -0.3282 0.1264>
   rotation=<rot -0.6523 -94.4602 -87.8941>
  }]
 dyn_ponytail_05 [{
   shapeType=Capsule
   length=3.7956
   radius=2.1664
   shapeName=dyn_braid_01_capsule
   center=<1.2556 0.5461 0.1233>
   rotation=<rot -1.0956 -81.8763 -88.0868>
  }]
 dyn_ponytail_06 [{
   shapeType=Capsule
   length=5.315
   radius=1.8858
   shapeName=dyn_ponytail_06_capsule
   center=<3.025 0.1769 0.0093>
   rotation=<rot -0.0021 -94.8552 -90.043>
  }]
 dyn_ponytail_base [{shapeType=Sphere, radius=0.505, shapeName=dyn_ponytail_base_sphere}]
 head [{shapeType=Sphere, radius=11.4327, shapeName=head_sphere, center=<5.803 2.4617 0.0032>}]
 neck_01 [{shapeType=Capsule, length=0, radius=6.9903, shapeName=neck_01_capsule, center=<3.3529 1.6803 0>, rotation=<rot 0 -90 0>}]
 spine_05 [
  {shapeType=Capsule, length=14.6354, radius=10.3696, shapeName=spine_05_capsule, center=<0.1827 1.7105 0>, rotation=<rot 0 -90 0>}
  {shapeType=Capsule, length=13.9681, radius=11.0369, shapeName=spine_05_capsule, center=<-9.2662 2.6428 0>, rotation=<rot 0 -90 0>}
 ]
 spine_cape_collision [
  {shapeType=Capsule, length=18.6618, radius=12.2108, shapeName=spine_05_capsule, center=<-5.6168 3.7299 0>, rotation=<rot 0 -90 0>}
  {shapeType=Capsule, length=13.9681, radius=11.0369, shapeName=spine_05_capsule, center=<-15.6236 1.9553 0>, rotation=<rot 0 -90 0>}
 ]
}
constraints [
 {bone1Name=dyn_ponytail_03, bone2Name=dyn_ponytail_02, swing1LimitDegrees=40, swing2LimitDegrees=15, twistLimitDegrees=0}
 {
  bone1Name=head
  bone2Name=neck_01
  swing1LimitDegrees=45
  swing1Motion=Free
  swing2LimitDegrees=45
  swing2Motion=Free
  twistLimitDegrees=45
  twistMotion=Free
 }
 {
  bone1Name=neck_01
  bone2Name=spine_05
  swing1LimitDegrees=45
  swing1Motion=Free
  swing2LimitDegrees=45
  swing2Motion=Free
  twistLimitDegrees=45
  twistMotion=Free
 }
 {bone1Name=dyn_ponytail_01, bone2Name=dyn_ponytail_base, swing1LimitDegrees=40, swing2LimitDegrees=15, twistLimitDegrees=0}
 {bone1Name=dyn_ponytail_02, bone2Name=dyn_ponytail_01, swing1LimitDegrees=40, swing2LimitDegrees=15, twistLimitDegrees=0}
 {bone1Name=dyn_ponytail_04, bone2Name=dyn_ponytail_03, swing1LimitDegrees=40, swing2LimitDegrees=15, twistLimitDegrees=0}
 {bone1Name=dyn_ponytail_05, bone2Name=dyn_ponytail_04, swing1LimitDegrees=40, swing2LimitDegrees=15, twistLimitDegrees=0}
 {bone1Name=dyn_ponytail_06, bone2Name=dyn_ponytail_05, swing1LimitDegrees=40, swing2LimitDegrees=15, twistLimitDegrees=0}
 {
  bone1Name=dyn_ponytail_base
  bone2Name=head
  swing1LimitDegrees=45
  swing1Motion=Free
  swing2LimitDegrees=45
  swing2Motion=Free
  twistLimitDegrees=45
  twistMotion=Free
 }
 {
  bone1Name=clavicle_l
  bone2Name=spine_05
  swing1LimitDegrees=45
  swing1Motion=Free
  swing2LimitDegrees=45
  swing2Motion=Free
  twistLimitDegrees=45
  twistMotion=Free
 }
 {
  bone1Name=clavicle_r
  bone2Name=spine_05
  swing1LimitDegrees=45
  swing1Motion=Free
  swing2LimitDegrees=45
  swing2Motion=Free
  twistLimitDegrees=45
  twistMotion=Free
 }
 {
  bone1Name=spine_cape_collision
  bone2Name=spine_05
  swing1LimitDegrees=45
  swing1Motion=Free
  swing2LimitDegrees=45
  swing2Motion=Free
  twistLimitDegrees=45
  twistMotion=Free
 }
]
